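# Physics asset for the Critter Maniac backpack cosmetic, giving its mesh collision bodies and jointed chains for six dangling strands.
# Paired capsules cover the pack and coffin bones, while small capsules follow each front, right and back strand.
# Each strand chains to a master body constrained to the pack, as is the coffin.
# Strand joints have modest swing and tight twist limits; only back strand 04's master explicitly locks rotation.
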PhysicsAsset /BRCosmetics/Accessories/FORT_Backpacks/Backpack_M_MED_Critter_Maniac/Meshes/M_MED_Critter_Maniac_Pack_Physics
bodies {
 coffin [
  {shapeType=Capsule, length=13.75, radius=7, shapeName=coffin_capsule, center=<-1.4352 -1.6152 2.1415>, rotation=<rot -47.768 0 -22.0454>}
  {shapeType=Capsule, length=15.75, radius=5, shapeName=coffin_capsule, center=<-13.4292 8.55 -9.7218>, rotation=<rot -47.7671 0 -22.0425>}
 ]
 dyn_back_strand_03_01 [{shapeType=Capsule, length=1, radius=0.5, shapeName=dyn_back_strand_03_01_capsule, center=<1.5 0 0>, rotation=<rot 90 0 0>}]
 dyn_back_strand_03_02 [{shapeType=Capsule, length=1.5, radius=0.5, shapeName=dyn_back_strand_03_02_capsule, center=<1.5 0 0>, rotation=<rot 90 0 0>}]
 dyn_back_strand_03_master [{shapeType=Capsule, length=0, radius=0.5, shapeName=dyn_back_strand_03_master_capsule}]
 dyn_back_strand_04_01 [{shapeType=Capsule, length=1, radius=0.5, shapeName=dyn_back_strand_04_01_capsule, center=<1.5 0 0>, rotation=<rot 90 0 0>}]
 dyn_back_strand_04_02 [{shapeType=Capsule, length=1.25, radius=0.5, shapeName=dyn_back_strand_04_02_capsule, center=<1 0 0>, rotation=<rot 90 0 0>}]
 dyn_back_strand_04_03 [{shapeType=Capsule, length=1.5, radius=0.5, shapeName=dyn_back_strand_04_03_capsule, center=<1 0 0>, rotation=<rot 90 0 0>}]
 dyn_back_strand_04_04 [{shapeType=Capsule, length=1, radius=0.5, shapeName=dyn_back_strand_04_04_capsule, center=<1.5 0 0>, rotation=<rot 90 0 0>}]
 dyn_back_strand_04_05 [{shapeType=Capsule, length=3, radius=0.5, shapeName=dyn_back_strand_04_05_capsule, center=<2 0 0>, rotation=<rot 90 0 0>}]
 dyn_back_strand_04_master [{shapeType=Capsule, length=0, radius=0.5, shapeName=dyn_back_strand_04_master_capsule}]
 dyn_back_strand_05_01 [{shapeType=Capsule, length=2, radius=0.5, shapeName=dyn_back_strand_05_01_capsule, center=<1.5 0 0>, rotation=<rot 90 0 0>}]
 dyn_back_strand_05_02 [{shapeType=Capsule, length=2, radius=0.5, shapeName=dyn_back_strand_05_02_capsule, center=<1.5 0 0>, rotation=<rot 90 0 0>}]
 dyn_back_strand_05_03 [{shapeType=Capsule, length=2, radius=0.5, shapeName=dyn_back_strand_05_03_capsule, center=<1.5 0 0>, rotation=<rot 90 0 0>}]
 dyn_back_strand_05_04 [{shapeType=Capsule, length=2, radius=0.5, shapeName=dyn_back_strand_05_04_capsule, center=<1.5 0 0>, rotation=<rot 90 0 0>}]
 dyn_back_strand_05_05 [{shapeType=Capsule, length=3, radius=0.5, shapeName=dyn_back_strand_05_05_capsule, center=<2 0 0>, rotation=<rot 90 0 0>}]
 dyn_back_strand_05_master [{shapeType=Capsule, length=0, radius=0.5, shapeName=dyn_back_strand_05_master_capsule}]
 dyn_back_strand_06_01 [{shapeType=Capsule, length=1, radius=0.5, shapeName=dyn_back_strand_06_01_capsule, center=<1 0 0>, rotation=<rot 90 0 0>}]
 dyn_back_strand_06_02 [{shapeType=Capsule, length=2, radius=0.5, shapeName=dyn_back_strand_06_02_capsule, center=<1.5 0 0>, rotation=<rot 90 0 0>}]
 dyn_back_strand_06_master [{shapeType=Capsule, length=0, radius=0.5, shapeName=dyn_back_strand_06_master_capsule}]
 dyn_front_strand_01_01 [{shapeType=Capsule, length=0.5, radius=0.5, shapeName=dyn_front_strand_01_01_capsule, center=<1 0 0>, rotation=<rot 90 0 0>}]
 dyn_front_strand_01_02 [{shapeType=Capsule, length=0.75, radius=0.5, shapeName=dyn_front_strand_01_02_capsule, center=<1 0 0>, rotation=<rot 90 0 0>}]
 dyn_front_strand_01_03 [{shapeType=Capsule, length=0.75, radius=0.5, shapeName=dyn_front_strand_01_03_capsule, center=<1 0 0>, rotation=<rot 90 0 0>}]
 dyn_front_strand_01_master [{shapeType=Capsule, length=0, radius=0.5, shapeName=dyn_front_strand_01_master_capsule}]
 dyn_right_strand_02_01 [{shapeType=Capsule, length=0.5, radius=0.5, shapeName=dyn_right_strand_02_01_capsule, center=<1.4 0 0>, rotation=<rot 90 0 0>}]
 dyn_right_strand_02_02 [{shapeType=Capsule, length=1.5, radius=0.5, shapeName=dyn_right_strand_02_02_capsule, center=<1.1 0 0>, rotation=<rot 90 0 0>}]
 dyn_right_strand_02_03 [{shapeType=Capsule, length=3, radius=0.5, shapeName=dyn_right_strand_02_03_capsule, center=<2 0 0>, rotation=<rot 90 0 0>}]
 dyn_right_strand_02_master [{shapeType=Capsule, length=0, radius=0.5, shapeName=dyn_right_strand_02_master_capsule}]
 pack [
  {shapeType=Capsule, length=36.5, radius=22.75, shapeName=pack_capsule, center=<4.2 24 -20>, rotation=<rot 90 0 0>}
  {shapeType=Capsule, length=36.5, radius=22.75, shapeName=pack_capsule, center=<4.2 24 0>, rotation=<rot 90 0 0>}
 ]
}
constraints [
 {bone1Name=dyn_right_strand_02_03, bone2Name=dyn_right_strand_02_02, swing1LimitDegrees=20, swing2LimitDegrees=20, twistLimitDegrees=1}
 {bone1Name=dyn_right_strand_02_01, bone2Name=dyn_right_strand_02_master, swing1LimitDegrees=20, swing2LimitDegrees=20, twistLimitDegrees=1}
 {bone1Name=dyn_right_strand_02_02, bone2Name=dyn_right_strand_02_01, swing1LimitDegrees=20, swing2LimitDegrees=20, twistLimitDegrees=1}
 {bone1Name=dyn_front_strand_01_03, bone2Name=dyn_front_strand_01_02, swing1LimitDegrees=25, swing2LimitDegrees=25, twistLimitDegrees=1}
 {bone1Name=dyn_front_strand_01_01, bone2Name=dyn_front_strand_01_master, swing1LimitDegrees=25, swing2LimitDegrees=25, twistLimitDegrees=1}
 {bone1Name=dyn_front_strand_01_02, bone2Name=dyn_front_strand_01_01, swing1LimitDegrees=25, swing2LimitDegrees=25, twistLimitDegrees=1}
 {bone1Name=dyn_back_strand_06_02, bone2Name=dyn_back_strand_06_01, swing1LimitDegrees=15, swing2LimitDegrees=15, twistLimitDegrees=1}
 {bone1Name=dyn_back_strand_06_01, bone2Name=dyn_back_strand_06_master, swing1LimitDegrees=15, swing2LimitDegrees=15, twistLimitDegrees=1}
 {bone1Name=dyn_back_strand_05_05, bone2Name=dyn_back_strand_05_04, swing1LimitDegrees=15, swing2LimitDegrees=15, twistLimitDegrees=1}
 {bone1Name=dyn_back_strand_05_01, bone2Name=dyn_back_strand_05_master, swing1LimitDegrees=15, swing2LimitDegrees=15, twistLimitDegrees=1}
 {bone1Name=dyn_back_strand_05_02, bone2Name=dyn_back_strand_05_01, swing1LimitDegrees=15, swing2LimitDegrees=15, twistLimitDegrees=1}
 {bone1Name=dyn_back_strand_05_03, bone2Name=dyn_back_strand_05_02, swing1LimitDegrees=15, swing2LimitDegrees=15, twistLimitDegrees=1}
 {bone1Name=dyn_back_strand_05_04, bone2Name=dyn_back_strand_05_03, swing1LimitDegrees=15, swing2LimitDegrees=15, twistLimitDegrees=1}
 {bone1Name=dyn_back_strand_04_05, bone2Name=dyn_back_strand_04_04, swing1LimitDegrees=15, swing2LimitDegrees=15, twistLimitDegrees=1}
 {bone1Name=dyn_back_strand_04_01, bone2Name=dyn_back_strand_04_master, swing1LimitDegrees=15, swing2LimitDegrees=15, twistLimitDegrees=1}
 {bone1Name=dyn_back_strand_04_02, bone2Name=dyn_back_strand_04_01, swing1LimitDegrees=15, swing2LimitDegrees=15, twistLimitDegrees=1}
 {bone1Name=dyn_back_strand_04_03, bone2Name=dyn_back_strand_04_02, swing1LimitDegrees=15, swing2LimitDegrees=15, twistLimitDegrees=1}
 {bone1Name=dyn_back_strand_04_04, bone2Name=dyn_back_strand_04_03, swing1LimitDegrees=15, swing2LimitDegrees=15, twistLimitDegrees=1}
 {bone1Name=dyn_back_strand_03_02, bone2Name=dyn_back_strand_03_01, swing1LimitDegrees=15, swing2LimitDegrees=15, twistLimitDegrees=1}
 {bone1Name=dyn_back_strand_03_01, bone2Name=dyn_back_strand_03_master, swing1LimitDegrees=15, swing2LimitDegrees=15, twistLimitDegrees=1}
 {bone1Name=coffin, bone2Name=pack, swing1LimitDegrees=1, swing2LimitDegrees=1, twistLimitDegrees=1}
 {bone1Name=dyn_back_strand_03_master, bone2Name=pack, swing1LimitDegrees=1, swing2LimitDegrees=1, twistLimitDegrees=1}
 {
  bone1Name=dyn_back_strand_04_master
  bone2Name=pack
  swing1LimitDegrees=1
  swing1Motion=Locked
  swing2LimitDegrees=1
  swing2Motion=Locked
  twistLimitDegrees=1
  twistMotion=Locked
 }
 {bone1Name=dyn_back_strand_05_master, bone2Name=pack, swing1LimitDegrees=1, swing2LimitDegrees=1, twistLimitDegrees=1}
 {bone1Name=dyn_back_strand_06_master, bone2Name=pack, swing1LimitDegrees=1, swing2LimitDegrees=1, twistLimitDegrees=1}
 {bone1Name=dyn_front_strand_01_master, bone2Name=pack, swing1LimitDegrees=1, swing2LimitDegrees=1, twistLimitDegrees=1}
 {bone1Name=dyn_right_strand_02_master, bone2Name=pack, swing1LimitDegrees=1, swing2LimitDegrees=1, twistLimitDegrees=1}
]
